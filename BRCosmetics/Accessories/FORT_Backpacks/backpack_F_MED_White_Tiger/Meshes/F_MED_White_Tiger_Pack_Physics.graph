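# Physics asset for the White Tiger back bling cosmetic, a backpack accessory with a simulated knife.
# It holds two capsule bodies: a small one on the pack bone and a longer, offset capsule on the dyn_knife bone.
# One constraint joins dyn_knife to pack with very tight swing and twist limits, so any knife motion stays a slight wobble.
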PhysicsAsset /BRCosmetics/Accessories/FORT_Backpacks/backpack_F_MED_White_Tiger/Meshes/F_MED_White_Tiger_Pack_Physics
bodies {
 dyn_knife [{shapeType=Capsule, length=18.2193, radius=3.524, center=<0.1 0 15>, rotation=<rot 90 0 0>}]
 pack [{shapeType=Capsule, length=3, radius=3}]
}
constraints [{bone1Name=dyn_knife, bone2Name=pack, swing1LimitDegrees=1.25, swing2LimitDegrees=3, twistLimitDegrees=1.25}]
