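# Physics asset for the F_MED_White_Tiger backpack cosmetic, giving its dangling cans and handle simulation bodies.
# A sphere on the pack bone anchors box bodies on three can bones and a handle bone, each constrained directly to the pack.
# The cans have tight angular limits for a slight jiggle, while the handle allows a much wider twist.
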PhysicsAsset /BRCosmetics/Accessories/FORT_Backpacks/backpack_F_MED_White_Tiger/Meshes/F_MED_White_Tiger_Pack_PhysicsAsset
bodies {
 dyn_can_a [{shapeType=Box, extentX=5.0137, extentY=6.2572, extentZ=10.4537, center=<-0.1812 -3.4044 0.8717>, rotation=<rot -0.1619 -90.035 -0.4056>}]
 dyn_can_b [{shapeType=Box, extentX=5.6002, extentY=7.0288, extentZ=10.6293, center=<-0.1146 -3.3346 0.8038>, rotation=<rot -1.7437 -90 0.4063>}]
 dyn_can_c [{shapeType=Box, extentX=7.3964, extentY=6.909, extentZ=10.7111, center=<-0.2656 -3.0981 1.0774>, rotation=<rot 4.2458 -89.9999 -0.754>}]
 dyn_handle [{shapeType=Box, extentX=10, extentY=3.5, extentZ=6, center=<0 0 -1.6942>}]
 pack [{shapeType=Sphere, radius=3.25}]
}
constraints [
 {bone1Name=dyn_can_a, bone2Name=pack, swing1LimitDegrees=5, swing2LimitDegrees=2, twistLimitDegrees=1}
 {bone1Name=dyn_can_b, bone2Name=pack, swing1LimitDegrees=5, swing2LimitDegrees=2, twistLimitDegrees=1}
 {bone1Name=dyn_can_c, bone2Name=pack, swing1LimitDegrees=5, swing2LimitDegrees=2, twistLimitDegrees=1}
 {bone1Name=dyn_handle, bone2Name=pack, swing1LimitDegrees=1, swing2LimitDegrees=10, twistLimitDegrees=50}
]
